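# Genie Smoke vehicle boost cosmetic: a Niagara system trailing smoke-like ribbons and sprites from the car's skeletal mesh sockets.
# Ribbon_Main spawns per-port, dual-renderer trails with painted-variant color curves feeding dynamic material parameters.
# The sprite emitter adds paint-tinted particles shaped by curl noise, drag and jitter.
# Both fade out inherited source movement over lifetime, while GameModeID switches velocities and ribbon spawn rates, and frontend preview tweaks ribbon alpha and width.
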NiagaraSystem /VehicleCosmetics/Boosts/Boost_GenieSmoke/FX/NS_Boost_GenieSmoke
dependencies {
 usedDataInterfaces [
  &NiagaraDataInterfaceColorCurve
  &NiagaraDataInterfaceArrayFloat
  &NiagaraDataInterfaceCurve
  &NiagaraDataInterfaceSkeletalMesh
  &NiagaraDataInterfaceArrayFloat3
 ]
 usedDynamicInputs [&SelectVectorFromArray]
 usedModules [
  &self:SystemSpawnScript.SystemScriptSource.SystemScriptGraph.NiagaraNodeAssignment_0.SetVariables_#1
  &self:SystemSpawnScript.SystemScriptSource.SystemScriptGraph.NiagaraNodeAssignment_1.SetVariables_#2
  &SystemState
  &self:SystemSpawnScript.SystemScriptSource.SystemScriptGraph.NiagaraNodeAssignment_2.SetVariables_#3
 ]
}
emitters {
 NE_VC_Boost_Sprite_Parent {
  emitterUpdateScript [
   &EmitterState()
   &VC_FilteredSocketCount()
   &SpawnRate(SpawnRate=50, Use Spawn Probability=false)
  ]
  particleSpawnScript [
   &VC_PaintedVariantsOverride(
    Black=<rgba 1 1 1 1>
    BurntSienna=<rgba 0.5 0.1058 0 1>
    Cobalt=<rgba 0.429 0.4488 1.5 1>
    Crimson=<rgba 1.5 0.3 0.3325 1>
    DefaultColor=<rgba 1 0.5972 1 1>
    ForestGreen=<rgba 0.3464 0.5 0.3541 1>
    Grey=<rgba 0.9 0.9 0.9 0.8>
    Lime=<rgba 0.25 0.5 0 1>
    Orange=<rgba 1 0.335 0.05 1>
    Pink=<rgba 1 0.5 0.8182 1>
    Purple=<rgba 0.6952 0.3438 1.2 1>
    SkyBlue=<rgba 0.2 0.7323 1 1>
   )
   &InitializeParticle(
    Color=$Particles.Color
    Lifetime=&RandomRangeFloat(Minimum=1, Maximum=1.5)
    Sprite Rotation Angle=&RandomRangeFloat(Minimum=0, Maximum=1)
    Sprite Rotation Mode=::NewEnumerator1
    Sprite Size Max=<0.5 0.5>
    Sprite Size Min=<0.18 0.18>
    Sprite Size Mode=::NewEnumerator2
    UsePositionOffset=false
    Write Material Random=true
   )
   &PartitionParticles(Number Of Partitions=$Emitter.PortCount)
   &SkeletalMeshLocation(
    Bone / Socket Index=$Particles.PartitionParticles.Partition
    Bone / Socket Sampling Mode=::NewEnumerator5
    MaskSpawnGroup=false
    Mesh Sampling Type=::NewEnumerator6
    Sampled Position Offset=&AddVectorToPosition(Position=<-20 0 -5>, Vector=$User.DTCarOffset)
    Skeletal Mesh=$System.SkeletalMesh
   )
   &AddVelocity(Velocity=&RandomRangeVector(
     Minimum=&MergeHelper_VectorSwitch(0=<-160 0 0>, 1=<-160 0 0>, 2=<-160 0 0>, 3=<-480 0 0>, Index=$User.GameModeID, Num=::NewEnumerator2)
     Maximum=&MergeHelper_VectorSwitch(0=<-400 0 0>, 1=<-400 0 0>, 2=<-400 0 0>, 3=<-2500 0 0>, Index=$User.GameModeID, Num=::NewEnumerator2)
    ))
   &Color(Color=$Particles.Color)
   /SetVariables(Particles.FakeGlow:Vector2f=<15 15>) {isSetParameters=true}
   &ShapeLocation(MaskSpawnGroup=false, Override Seed=false, Sphere Radius=50, UseOwnerScale=true)
   /VehicleCosmetics/FX/VC_FXParents/Modules/DLR_ProductPreviews(GameModeID:NiagaraInt32=$System.GameModeID, bPreviewVehicleSpeed:NiagaraBool=$System.bPreviewVehicleSpeed)
  ]
  particleUpdateScript [
   &ParticleState()
   &CurlNoiseForce(
    Bias Noise=false
    Noise Frequency=50000
    Noise Strength=&RandomRangeFloat(Minimum=200, Maximum=300)
    Pan Noise=false
    Randomization Vector:Vector3f=<0.65 0.125 0>
   )
   &InheritSourceMovement(
    Applied Position Scale=&VectorFromFloat(Value=&FloatFromCurve(FloatCurve=DataInterface() {propertyValues {Curve {keys [{interpMode=RCIM_Cubic, tangentMode=RCTM_Break, time=0.3, value=1}, {interpMode=RCIM_Cubic, tangentMode=RCTM_Break, time=1, value=0}]}}}))
    ApplyForce=false
    ApplyPosition=true
    ApplyPositionFalloff=false
   ) {enabled=false}
   &ScaleSpriteSize(Scale Sprite Size Mode=::NewEnumerator0, Uniform Scale Factor=12)
   &Drag(
    Drag=&RandomRangeFloat(
     Minimum=0
     Maximum=&FloatFromCurve(FloatCurve=DataInterface() {propertyValues {Curve {defaultValue=1, keys [<at 0.2 0 RCIM_Cubic>,<at 1 1 RCIM_Cubic>]}}}, Scale Curve=2)
    )
    Use Linear Drag=true
    Use Rotational Drag=true
   )
   &InheritSourceMovement(
    Applied Position Scale=&Multiply_VectorByFloat(
     Vector=<1 1 1>
     Float=&FloatFromCurve(FloatCurve=DataInterface() {propertyValues {Curve {
         defaultValue=1
         keys [
          {arriveTangent=-1.7885, interpMode=RCIM_Cubic, leaveTangent=-1.7885, tangentMode=RCTM_User, time=0.05, value=1}
          {interpMode=RCIM_Cubic, tangentMode=RCTM_User, time=1, value=0}
         ]
        }}})
    )
    ApplyForce=false
    ApplyPosition=true
    ApplyPositionFalloff=false
   ) {module=InheritSourceMovement001}
   &ScaleColor(
    Scale Alpha=1.5
    Scale RGB=&VectorFromFloat(Value=&MakeCustomFloatFromBool(Bool=$User.bIsPainted, True Float=15, False Float=15))
    ScaleA=true
    ScaleRGB=true
   )
   &JitterPosition(Jitter Amount:NiagaraFloat=&RandomRangeFloat(Minimum=-4, Maximum=4), Jitter Delay=2)
   &SolveForcesAndVelocity(Clamp Velocity=false, Limit Acceleration=false)
  ]
  rendererClasses [&NiagaraSpriteRendererProperties]
  renderers [
   {rendererClass=&NiagaraSpriteRendererProperties}
   {rendererClass=&NiagaraSpriteRendererProperties, rendererIndex=1}
  ]
 }
 Ribbon_Main {
  emitterUpdateScript [
   &EmitterState()
   &VC_FilteredSocketCount()
   &VC_RateQualityHandler(
    Epic Spawn Rate:NiagaraFloat=&MergeHelper_FloatSwitch(0=30, 1=30, 2=30, 3=10, Index=$User.GameModeID, Num=::NewEnumerator2)
    High Spawn Rate:NiagaraFloat=&MergeHelper_FloatSwitch(0=60, 1=60, 2=60, 3=10, Index=$User.GameModeID, Num=::NewEnumerator2)
    Low Spawn Rate:NiagaraFloat=&MergeHelper_FloatSwitch(0=30, 1=30, 2=30, 3=10, Index=$User.GameModeID, Num=::NewEnumerator2)
    Medium Spawn Rate:NiagaraFloat=&MergeHelper_FloatSwitch(0=30, 1=30, 2=30, 3=10, Index=$User.GameModeID, Num=::NewEnumerator2)
    Use Custom Rate?=true
   )
   &VC_SpawnRibbon_v1(Number of Ribbons=$Emitter.PortCount, SpawnRate=$Emitter.Spawn Rate)
  ]
  particleSpawnScript [
   &VC_PaintedVariantsOverride(R1 TitaniumWhite:LinearColor=<rgba 1 0 0 1>, Use R1 Painted Overrides=true) {enabled=false}
   &VC_InitializeRibbon_v1(
    Color=$Particles.Color
    Custom Offset=true
    Enable Skeletal Position=true
    Lifetime=0.8
    Local Vector Offset=&AddVectorToPosition(Position=<0 0 0>, Vector=$User.DTCarOffset)
    Offset Amount=1
    Ribbon Width=100
   )
   &AddVelocity(Velocity=&MergeHelper_VectorSwitch(0=<-1600 0 0>, 1=<-1000 0 0>, 2=<-1000 0 0>, 3=<-2500 0 0>, Index=$User.GameModeID, Num=::NewEnumerator2)) {module=AddVelocity001}
   /SetVariables(Particles.WidthComple=130) {isSetParameters=true}
   &DLR_ProductPreviews(GameModeID:NiagaraInt32=$System.GameModeID, bPreviewVehicleSpeed:NiagaraBool=$System.bPreviewVehicleSpeed)
  ]
  particleUpdateScript [
   &ParticleState()
   &VC_PaintedVariantsOverride(
    BurntSienna=<rgba 0.0642 0.0198 0.0079 1>
    Crimson=<rgba 1 0 0.0271 1>
    R1 Black:LinearColor=&ColorFromCurve(ColorCurve=DataInterface() {propertyValues {
       AlphaCurve {keys [<at 0 0.8 RCIM_Cubic>, <at 0.2107 0.45>, {interpMode=RCIM_Cubic, leaveTangent=-0.5701, time=0.7628, value=1}, <at 0.9774 0>]}
       BlueCurve {keys [
         <at 0.0008 0.35>
         {arriveTangent=-0.4533, interpMode=RCIM_Cubic, leaveTangent=-0.4533, time=0.4628, value=0.085}
         <at 0.5524 0.1>
         <at 0.8065 0.1>
         <at 0.9295 0.3>
        ]}
       GreenCurve {keys [
         <at 0.0008 0.35>
         {arriveTangent=-0.4533, interpMode=RCIM_Cubic, leaveTangent=-0.4533, time=0.4628, value=0.085}
         <at 0.5524 0.1>
         <at 0.8065 0.1>
         <at 0.9295 0.3>
        ]}
       RedCurve {keys [
         <at 0.0008 0.35>
         {arriveTangent=-0.4533, interpMode=RCIM_Cubic, leaveTangent=-0.4533, time=0.4628, value=0.085}
         <at 0.5524 0.1>
         <at 0.8065 0.1>
         <at 0.9295 0.3>
        ]}
      }})
    R1 BurntSienna:LinearColor=&ColorFromCurve(ColorCurve=DataInterface() {propertyValues {
       AlphaCurve {keys [<at 0 0.8 RCIM_Cubic>,<at 0.2253 0.5>,<at 0.419 0.5>,<at 0.642 1.0756>,<at 1 5>]}
       BlueCurve {keys [<at 0.0315 0.1275 RCIM_Cubic>,<at 0.6336 0.07>,<at 0.7378 0.06>,<at 1 0.1 RCIM_Cubic>]}
       GreenCurve {keys [<at 0.0315 0.1877 RCIM_Cubic>,<at 0.6336 0.0933>,<at 0.7378 0.08>,<at 1 0.175 RCIM_Cubic>]}
       RedCurve {keys [<at 0.0315 0.85 RCIM_Cubic>,<at 0.6336 0.35>,<at 0.7378 0.3>,<at 1 1 RCIM_Cubic>]}
      }})
    R1 Cobalt:LinearColor=&ColorFromCurve(ColorCurve=DataInterface() {propertyValues {
       AlphaCurve {keys [
         <at 0 0.8 RCIM_Cubic>
         {arriveTangent=-1.5851, interpMode=RCIM_Cubic, leaveTangent=-1.5851, time=0, value=0.8}
         <at 0.2524 0.4>
         <at 0.844 0.7322>
         <at 0.9961 0>
        ]}
       BlueCurve {keys [<at 0 2 RCIM_Cubic>,<at 0.5753 1>,<at 0.769 0.62>,<at 1 2>]}
       GreenCurve {keys [<at 0 0.344 RCIM_Cubic>,<at 0.5753 0.16>,<at 0.769 0.0931>,<at 1 0.2827>]}
       RedCurve {keys [<at 0 0.16 RCIM_Cubic>,<at 0.5753 0.1>,<at 0.769 0.0554>,<at 1 0.16>]}
      }})
    R1 Crimson:LinearColor=&ColorFromCurve(ColorCurve=DataInterface() {propertyValues {
       AlphaCurve {keys [<at 0 1 RCIM_Cubic>,<at 0.644 1>,<at 1 5 RCIM_Cubic>]}
       BlueCurve {keys [<at 0 0.1947 RCIM_Cubic>,<at 0.6357 0.0055>,<at 0.7732 0.0396>,<at 1 0.1>]}
       GreenCurve {keys [<at 0 0.15 RCIM_Cubic>,<at 0.6357 0.0055>,<at 0.7732 0.0396>,<at 1 0.1>]}
       RedCurve {keys [<at 0 1.5 RCIM_Cubic>,<at 0.6357 0.55>,<at 0.7732 0.5>,<at 1 1>]}
      }})
    R1 DefaultColor=&ColorFromCurve(ColorCurve=DataInterface() {propertyValues {
       AlphaCurve {keys [<at 0 0.25>, <at 0.3877 0.7>, {arriveTangent=-1.6286, interpMode=RCIM_Cubic, leaveTangent=-1.6286, time=0.6142, value=0.8}, <at 0.8175 0>]}
       BlueCurve {keys [<at 0.4376 1 RCIM_Cubic>,<at 0.723 1.8>,<at 1 1 RCIM_Cubic>]}
       GreenCurve {keys [<at 0.4376 1 RCIM_Cubic>,<at 0.723 1.8>,<at 1 1 RCIM_Cubic>]}
       RedCurve {keys [<at 0.4376 1 RCIM_Cubic>,<at 0.723 1.8>,<at 1 1 RCIM_Cubic>]}
      }})
    R1 ForestGreen:LinearColor=&ColorFromCurve(ColorCurve=DataInterface() {propertyValues {
       AlphaCurve {keys [<at 0.6878 1 RCIM_Cubic>,<at 1 5>]}
       BlueCurve {keys [<at 0.0544 0.1161>,<at 0.3315 0.0247>,<at 0.692 0.0106>,<at 0.8774 0.015>,<at 0.9524 0.01>]}
       GreenCurve {keys [<at 0.0544 0.8>,<at 0.3315 0.35>,<at 0.692 0.15>,<at 0.8774 0.3>,<at 0.9524 0.2>]}
       RedCurve {keys [<at 0.0544 0.08>,<at 0.3315 0.0072>,<at 0.692 0.0031>,<at 0.8774 0>,<at 0.9524 0>]}
      }})
    R1 Grey:LinearColor=&ColorFromCurve(ColorCurve=DataInterface() {propertyValues {
       AlphaCurve {keys [<at 0 1 RCIM_Cubic>,<at 0.6774 1>,<at 0.8878 3.6089>,<at 1 0 RCIM_Cubic>]}
       BlueCurve {keys [<at 0 0.5 RCIM_Cubic>,<at 0.4862 0.25>,<at 0.7253 0.15>,<at 1 0.4>]}
       GreenCurve {keys [<at 0 0.5 RCIM_Cubic>,<at 0.4862 0.25>,<at 0.7253 0.15>,<at 1 0.4>]}
       RedCurve {keys [<at 0 0.5 RCIM_Cubic>,<at 0.4862 0.25>,<at 0.7253 0.15>,<at 1 0.4>]}
      }})
    R1 Lime:LinearColor=&ColorFromCurve(ColorCurve=DataInterface() {propertyValues {
       AlphaCurve {keys [<at 0 1 RCIM_Cubic>,<at 0.7899 1>,<at 1 5 RCIM_Cubic>]}
       BlueCurve {keys [<at 0 0>,<at 0.4296 0>,<at 0.7899 0>,<at 1 0>]}
       GreenCurve {keys [<at 0 0.9>,<at 0.4296 0.35>,<at 0.7899 0.18>,<at 1 0.5>]}
       RedCurve {keys [<at 0 0.4513>,<at 0.4296 0.1755>,<at 0.7899 0.0903>,<at 1 0.2507>]}
      }})
    R1 Orange:LinearColor=&ColorFromCurve(ColorCurve=DataInterface() {propertyValues {
       AlphaCurve {keys [<at 0 0.5 RCIM_Cubic>,<at 0.4336 0.4>,<at 0.7295 1>,<at 1 5 RCIM_Cubic>]}
       BlueCurve {keys [<at 0 0.033>,<at 0.4312 0.014>,<at 0.7711 0.0044>,<at 1 0>]}
       GreenCurve {keys [<at 0 0.2486>,<at 0.4312 0.1055>,<at 0.7711 0.0704>,<at 1 0.16>]}
       RedCurve {keys [<at 0 1.65>,<at 0.4312 0.7>,<at 0.7711 0.5>,<at 1 1.2>]}
      }})
    R1 Pink:LinearColor=&ColorFromCurve(ColorCurve=DataInterface() {propertyValues {
       AlphaCurve {keys [
         <at 0 0.5 RCIM_Cubic>
         <at 0.3003 0.4>
         {arriveTangent=6.6304, interpMode=RCIM_Cubic, leaveTangent=6.6304, time=0.7345, value=1}
         <at 0.9941 5>
        ]}
       BlueCurve {keys [<at 0 0.9918 RCIM_Cubic>,<at 0.469 0.494>,<at 0.7586 0.2871>,<at 1 0.6212>]}
       GreenCurve {keys [<at 0 0.28 RCIM_Cubic>,<at 0.469 0.13>,<at 0.7586 0.0679>,<at 1 0.135>]}
       RedCurve {keys [<at 0 1.4 RCIM_Cubic>,<at 0.469 0.65>,<at 0.7586 0.4>,<at 1 0.9>]}
      }})
    R1 Purple:LinearColor=&ColorFromCurve(ColorCurve=DataInterface() {propertyValues {
       AlphaCurve {keys [<at 0 1 RCIM_Cubic>,<at 0.3649 1>,<at 0.7357 1>,<at 1 2 RCIM_Cubic>]}
       BlueCurve {keys [
         <at 0 3 RCIM_Cubic>
         {arriveTangent=-2.5728, interpMode=RCIM_Cubic, leaveTangent=-2.5728, time=0.4357, value=1.4}
         <at 0.6149 1>
         <at 0.7774 1>
         <at 1 3>
        ]}
       GreenCurve {keys [
         <at 0 0 RCIM_Cubic>
         {arriveTangent=0.0257, interpMode=RCIM_Cubic, leaveTangent=0.0257, time=0.4357, value=0}
         <at 0.6149 0.0055>
         <at 0.7774 0.02>
         <at 1 0>
        ]}
       RedCurve {keys [
         <at 0 0.5 RCIM_Cubic>
         {arriveTangent=-0.4093, interpMode=RCIM_Cubic, leaveTangent=-0.4093, time=0.4357, value=0.2312}
         <at 0.6149 0.1795>
         <at 0.7774 0.1818>
         <at 1 0.75>
        ]}
      }})
    R1 Saffron:LinearColor=&ColorFromCurve(ColorCurve=DataInterface() {propertyValues {
       AlphaCurve {keys [<at 0 0.5 RCIM_Cubic>,<at 0.2253 0.4>,<at 0.7274 3.6646>,<at 0.9961 5>]}
       BlueCurve {keys [<at 0 0.033>,<at 0.4053 0.0123>,<at 0.6149 0.0076>,<at 0.8232 0.0092>,<at 1 0>]}
       GreenCurve {keys [<at 0 0.7443>,<at 0.4053 0.305>,<at 0.6149 0.1897>,<at 0.8232 0.2276>,<at 1 0.5333>]}
       RedCurve {keys [<at 0 1.1>,<at 0.4053 0.402>,<at 0.6149 0.25>,<at 0.8232 0.3>,<at 1 0.8>]}
      }})
    R1 SkyBlue:LinearColor=&ColorFromCurve(ColorCurve=DataInterface() {propertyValues {
       AlphaCurve {keys [<at 0 1 RCIM_Cubic>, {arriveTangent=-1, interpMode=RCIM_Cubic, leaveTangent=-1, time=0, value=0.5}, <at 0.8107 0.75>, <at 1 0>]}
       BlueCurve {keys [<at 0 1.3>,<at 0.3586 0.625>,<at 0.6586 0.425>,<at 0.817 0.4>,<at 0.8899 0.8>,<at 1 0>]}
       GreenCurve {keys [<at 0 0.7583>,<at 0.3586 0.3508>,<at 0.6586 0.217>,<at 0.817 0.2141>,<at 0.8899 0.5067>,<at 1 0>]}
       RedCurve {keys [<at 0 0>,<at 0.3586 0>,<at 0.6586 0>,<at 0.817 0>,<at 0.8899 0>,<at 1 0>]}
      }})
    R1 TitaniumWhite:LinearColor=&ColorFromCurve(ColorCurve=DataInterface() {propertyValues {
       AlphaCurve {keys [<at 0 0.5 RCIM_Cubic>,<at 0.3815 0.7>,<at 1 5 RCIM_Cubic>]}
       BlueCurve {keys [<at 0 1 RCIM_Cubic>,<at 0.3774 0.35>,<at 0.6878 0.2>,<at 0.8295 0.2>,<at 1 0.5>]}
       GreenCurve {keys [<at 0 1 RCIM_Cubic>,<at 0.3774 0.35>,<at 0.6878 0.2>,<at 0.8295 0.2>,<at 1 0.5>]}
       RedCurve {keys [<at 0 1 RCIM_Cubic>,<at 0.3774 0.35>,<at 0.6878 0.2>,<at 0.8295 0.2>,<at 1 0.5>]}
      }})
    Use R1 Painted Overrides=true
   ) {module=VC_PaintedVariantsOverride002}
   &DynamicMaterialParameters(
    Index 0 Param 1=&MakeCustomFloatFromBool(Bool=$User.bIsPainted, True Float=1, False Float=0)
    Index 0 Param 2=&MakeCustomFloatFromBool(Bool=$User.bIsPainted, True Float=0, False Float=3)
    Index 0 Param 3=&SelectFloatFromArray(Direct Array Index=&SpawnGroup(), Float Selection Array=DataInterface() {propertyValues {FloatData [0.3,-0.3]}})
    Index 0 Param 4=&FloatFromCurve(
     FloatCurve=DataInterface() {propertyValues {Curve {
        defaultValue=1
        keys [<at 0 0.5 RCIM_Cubic>, {arriveTangent=1.7103, interpMode=RCIM_Cubic, leaveTangent=1.7103, tangentMode=RCTM_User, time=1, value=1}]
       }}}
     Scale Curve=0.3
    )
    Index 1 Linear Color:LinearColor=&MakeCustomLinearColorFromBool(Bool=$User.bIsPainted, True Color=$Particles.R1 Color, False Color=<rgba 0.01 0.2 1 1>)
    Index 2 Linear Color:LinearColor=&MakeCustomLinearColorFromBool(Bool=$User.bIsPainted, True Color=$Particles.R1 Color, False Color=<rgba 1.2 0.156 1.2 1>)
    Module.Index0Float0Write=true
    Module.Index0Float1Write=true
    Module.Index0Float2Write=true
    Module.Index0Float3Write=true
    Write Parameter Index 1=::NewEnumerator5
    Write Parameter Index 2=::NewEnumerator5
   )
   /SetVariables(
    Particles.Dynamic_CompleTrail:Vector4f=&MakeVector4(
     X=&SelectFloatFromArray(Direct Array Index=&SpawnGroup(), Float Selection Array=DataInterface() {propertyValues {FloatData [0.3,0]}})
     Y=4
     Z=&MakeCustomFloatFromBool(Bool=$User.bIsPainted, True Float=1, False Float=0)
     W=1.1
    )
    Particles.ScaleWidthComple:NiagaraFloat=&Multiply_Float(
     A=$Particles.WidthComple
     B=&FloatFromCurve(FloatCurve=DataInterface() {propertyValues {Curve {keys [{interpMode=RCIM_Cubic, tangentMode=RCTM_Break, time=0.2, value=1}, {interpMode=RCIM_Cubic, tangentMode=RCTM_Break, time=1, value=0}]}}})
    )
    Particles.Scale_ColorComple:LinearColor=&VC_DynamicScaleColor(
     Color to Scale:LinearColor=$Particles.R1 Color
     RGB Scale:Vector3f=&VectorFromFloat(Value=&MakeCustomFloatFromBool(
       Bool=$User.bIsPainted
       True Float=&FloatFromCurve(FloatCurve=DataInterface() {propertyValues {Curve {keys [<at 0 0.3>,<at 0.5 1>,<at 1 0.5>]}}}, Scale Curve=2.5)
       False Float=1
      ))
     Alpha Scale:NiagaraFloat=1
    )
   ) {isSetParameters=true}
   &InheritSourceMovement(
    Applied Position Scale=&VectorFromFloat(Value=1)
    ApplyForce=false
    ApplyPosition=true
    ApplyPositionFalloff=false
   ) {enabled=false, module=InheritSourceMovement001}
   &InheritSourceMovement(
    Applied Position Scale=&VectorFromFloat(Value=&FloatFromCurve(FloatCurve=DataInterface() {propertyValues {Curve {
         defaultValue=1
         keys [
          {arriveTangent=-1.7885, interpMode=RCIM_Cubic, leaveTangent=-1.7885, tangentMode=RCTM_User, time=0.05, value=1}
          {interpMode=RCIM_Cubic, tangentMode=RCTM_User, time=1, value=0}
         ]
        }}}))
    ApplyForce=false
    ApplyPosition=true
    ApplyPositionFalloff=false
   )
   &SolveForcesAndVelocity(Clamp Velocity=false, Limit Acceleration=false)
   &ScaleColor(
    Scale Alpha=&MakeCustomFloatFromBool(Bool=$User.bIsFrontendPreview, True Float=0.5, False Float=1)
    Scale RGB=&MakeCustomVectorFromBool(Bool=$User.bIsPainted, True Vector=<1 1 1>, False Vector=<2 1.3 1.3>)
    ScaleA=true
    ScaleRGB=true
   )
   &ScaleRibbonWidth(Ribbon Width Scale=&FloatFromCurve(
     FloatCurve=DataInterface() {propertyValues {Curve {keys [{interpMode=RCIM_Cubic, tangentMode=RCTM_Break, time=0.2, value=1}, {interpMode=RCIM_Cubic, tangentMode=RCTM_Break, time=1, value=0.2}]}}}
     Scale Curve=&MakeCustomFloatFromBool(Bool=$User.bIsFrontendPreview, True Float=0.95, False Float=1.1)
    ))
   /SetVariables(Particles.R1 Color:LinearColor=$Particles.R1 Color) {isSetParameters=true}
   &VC_UpdateRibbon_v1(Enable Skeletal Position=true)
  ]
  rendererClasses [&NiagaraRibbonRendererProperties]
  renderers [
   {rendererClass=&NiagaraRibbonRendererProperties}
   {rendererClass=&NiagaraRibbonRendererProperties, rendererIndex=1}
  ]
 }
}
summary {
 systemName=NS_Boost_GenieSmoke
 emitters [
  {emitterName=Ribbon_Main, rendererClasses [&NiagaraRibbonRendererProperties]}
  {emitterName=NE_VC_Boost_Sprite_Parent, rendererClasses [&NiagaraSpriteRendererProperties]}
 ]
 userVariables [
  User.bIsFrontendPreview {type=NiagaraBool}
  User.bIsPainted {type=NiagaraBool}
  User.blsIsCoroColor {type=NiagaraBool}
  User.bPreviewVehicleSpeed {type=NiagaraBool}
  User.ColorVariantID {type=NiagaraInt32}
  User.DTCarOffset {type=Vector3f, description="", defaultValue {struct=Vector3f, value=<0 0 0>}}
  User.GameModeID {type=NiagaraInt32, defaultValue=3}
  User.PaintedColor {type=LinearColor, description="", defaultValue {struct=LinearColor, value=<rgba 1 1 1 1>}}
  User.PaintedScalar {type=NiagaraFloat, defaultValue=1, description=""}
  User.PortCount {type=NiagaraInt32, defaultValue=2}
 ]
}
