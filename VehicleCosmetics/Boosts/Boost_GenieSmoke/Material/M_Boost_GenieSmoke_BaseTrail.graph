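# Unlit translucent Niagara material for the base trail of the Genie Smoke vehicle boost, usable on ribbons, sprites and mesh particles.
# Panning smoke trail textures, warped by scrolling RGB noise with a time-pulsed caustics layer, combine with cloud noise and floored noise-spot sparkles.
# Dynamic parameters drive a blue-to-pink gradient along U, plus desaturation, offset and distortion.
# Particle color tints the emissive, while its alpha scales opacity that fades at both trail ends.
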Material /VehicleCosmetics/Boosts/Boost_GenieSmoke/Material/M_Boost_GenieSmoke_BaseTrail
graph {
 properties {
  bAutomaticallySetUsageInEditor=false
  bUsedWithNiagaraMeshParticles=true
  bUsedWithNiagaraRibbons=true
  bUsedWithNiagaraSprites=true
  blendMode=BLEND_Translucent
  shadingModel=MSM_Unlit
  usedShadingModels=Unlit
 }
 nodes [
  n1 = Constant() {r=0.2}
  n2 = DynamicParameter() {parameterIndex=1, defaultValue=<rgba 0.01 0.2 1 1>, paramNames [R_BlueColor,G_BlueColor,B_BlueColor,Param4]}
  n3 = ComponentMask(None=TextureCoordinate()) {r=true}
  n4 = DynamicParameter() {defaultValue=<rgba 0 1 1 0.15>, paramNames [DesaturationRibbon, PowerDarkColor, Offsets, Distorsion]}
  n5 = Power(Base=OneMinus(None=@n3), Exp=@n4.PowerDarkColor)
  n6 = DynamicParameter() {parameterIndex=2, defaultValue=<rgba 1 0.15 1 1>, paramNames [R_PinkColor,G_PinkColor,B_PinkColor,Param4]}
  n7 = Desaturation(None=Add(A=Multiply(A=@n2.RGB, B=@n3), B=Multiply(A=@n5, B=@n6.RGB)), Fraction=@n4.DesaturationRibbon)
  n8 = ScalarParameter() {defaultValue=5, parameterName=RepWavesTrail}
  n9 = Constant() {r=1}
  n10 = AppendVector(A=@n4.Offsets, B=Constant() {r=-0.08})
  n11 = Add(A=Multiply(A=TextureCoordinate(), B=AppendVector(A=@n8, B=@n9)), B=@n10)
  n12 = Add(A=@n10, B=TextureCoordinate() {unMirrorU=true, unMirrorV=true})
  n13 = ComponentMask(None=TextureSample(UVs=Panner(Coordinate=@n12) {speedX=-0.29, speedY=0.09}) {texture=/DealershipSharedLibrary/FX/Textures/Noise/T_VCVFX_Alpine_Infinite_RGBNoise01}.RGB) {g=true, r=true}
  n14 = Multiply(A=@n13, B=@n4.Distorsion)
  n15 = Panner(Coordinate=Add(A=@n11, B=@n14), Speed=VectorParameter() {parameterName=Tex_Panner_MAIN, defaultValue=<rgba -0.5 0 0 1>}.RGB) {speedX=-1.5}
  n16 = TextureSample(UVs=@n15) {texture=/VehicleCosmetics/Boosts/Boost_GenieSmoke/Texture/T_VCVFX_Boost_GenieSmoke_Trail3_A}
  n17 = Constant() {r=30}
  n18 = Saturate(None=Multiply(A=Multiply(A=@n1, B=@n7), B=Multiply(A=@n16.RGB, B=@n17)))
  n19 = Constant() {r=3.5}
  n20 = Time()
  n21 = Multiply(A=LinearInterpolate(A=Constant() {r=0.15}, B=Constant() {r=0.2}, Alpha=Sine(None=@n20) {period=2}), B=@n13)
  n22 = TextureSample(UVs=Panner(Coordinate=Add(A=@n11, B=@n21), Speed=VectorParameter() {parameterName=Tex_Panner, defaultValue=<rgba -0.5 0 0 1>}.RGB) {speedX=-1.5}) {samplerSource=SSM_Wrap_WorldGroupSettings, texture=/VehicleCosmetics/Boosts/Boost_GenieSmoke/Texture/T_VCVFX_Boost_GenieSmoke_Trail2_RGB}
  n23 = Multiply(A=@n22.G, B=ScalarParameter() {defaultValue=1.1, parameterName=CAUSTICS_OPACITY}) {constB=1.1}
  n24 = TextureCoordinate() {uTiling=2}
  n25 = Panner(Coordinate=Add(A=Add(A=@n10, B=@n24), B=@n14), Speed=VectorParameter() {parameterName=noise_Panner, defaultValue=<rgba -0.4 0 0 1>}.RGB) {speedX=-0.4}
  n26 = TextureSample(UVs=@n15) {texture=&T_VCVFX_Boost_GenieSmoke_Trail2_RGB}
  n27 = Saturate(None=@n26.R)
  n28 = Add(A=Multiply(A=Multiply(A=@n19, B=@n23), B=TextureSample(UVs=@n25) {texture=/VehicleCosmetics/FX/Textures/Noise/T_VCVFX_Clouds_A}.R), B=@n27)
  n29 = TextureCoordinate() {uTiling=7, unMirrorU=true, vTiling=7}
  n30 = AppendVector(A=Constant() {r=1}, B=Constant())
  n31 = Constant() {r=1}
  n32 = Constant() {r=0.1}
  n33 = Panner(Coordinate=Multiply(A=Add(A=@n29, B=@n30), B=AppendVector(A=@n31, B=@n32))) {speedX=-1, speedY=0.05}
  n34 = Multiply(
   A=Desaturation(None=TextureSample(UVs=@n33) {texture=/DealershipSharedLibrary/FX/Textures/Noise/T_VCVFX_NoiseSpots_RGB}.R)
   B=ScalarParameter() {defaultValue=50, parameterName=SandIntensity}
  )
  n35 = Multiply(A=@n7, B=Add(A=Multiply(A=@n28, B=@n26.R), B=Floor(None=@n34)))
  MaterialExpressionVertexColor_0 = ParticleColor()
  MaterialFunctionCall(Emissive=Multiply(A=Add(A=@n18, B=@n35), B=@MaterialExpressionVertexColor_0.RGB)) {materialFunction=&MF_FX_EmissiveAdjustments}
  n36 = SmoothStep(Value=@n26.R) {constMin=0.6}
  n37 = ScalarParameter() {defaultValue=0.5, parameterName=MAINRIBBON_OPACITY}
  n38 = OneMinus(None=ComponentMask(None=TextureCoordinate()) {r=true})
  n39 = ComponentMask(None=TextureCoordinate()) {r=true}
  n40 = Power(Base=Saturate(None=Multiply(A=@n39) {constB=2}), Exp=ScalarParameter() {defaultValue=2, parameterName=Initial fade}) {constExponent=1}
  n41 = Add(A=Multiply(A=Multiply(A=@n36, B=@n37), B=@n27), B=Multiply(A=Multiply(A=@n36, B=@n23), B=Multiply(A=@n38, B=@n40)))
  n42 = OneMinus(None=@n39)
  n43 = ScalarParameter() {defaultValue=2.2, parameterName=power fade one minus}
  n44 = Multiply(A=Multiply(A=Power(Base=@n42, Exp=@n43), B=@n40), B=Constant() {r=50})
  Saturate(None=Multiply(A=Multiply(A=@n41, B=@n44), B=@MaterialExpressionVertexColor_0.A))
  ScalarParameter() {defaultValue=0.3, parameterName=DistPower}
 ]
}
